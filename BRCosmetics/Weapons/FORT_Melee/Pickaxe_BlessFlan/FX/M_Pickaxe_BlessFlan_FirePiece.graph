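# Translucent, unlit, two-sided fire piece material for BlessFlan pickaxe FX on Niagara meshes and ribbons.
# Emissive blends two colors along V, modulated by panning plasma textures and particle color.
# Opacity combines noise-distorted, counter-rotated textures and masks split at the V midpoint, shaped by UV edge, center and sphere masks, particle alpha, optional corner mask and view-angle fade.
# World position offset adds U-weighted panning noise to optional SplineThicken, both zero-weighted by default.
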Material /BRCosmetics/Weapons/FORT_Melee/Pickaxe_BlessFlan/FX/M_Pickaxe_BlessFlan_FirePiece
graph {
 properties {
  bAutomaticallySetUsageInEditor=false
  bUsedWithNiagaraMeshParticles=true
  bUsedWithNiagaraRibbons=true
  blendMode=BLEND_Translucent
  shadingModel=MSM_Unlit
  twoSided=true
  usedShadingModels=Unlit
 }
 nodes [
  n1 = ScalarParameter() {defaultValue=0.4, parameterName=SplineBase}
  n2 = ScalarParameter() {defaultValue=0.4, parameterName=SplineTip}
  n3 = StaticBool()
  n4 = TextureCoordinate()
  n5 = VectorParameter() {group=WPO, parameterName=WPO Side UV, sortPriority=6, defaultValue=<rgba 1.5 1 -0.3 0.3>}
  n6 = AppendVector(A=@n5.R, B=@n5.R)
  n7 = TextureSample(UVs=Panner(Coordinate=Multiply(A=@n4, B=@n6), Speed=AppendVector(A=@n5.B, B=@n5.B))) {samplerType=SAMPLERTYPE_Normal, texture=/Game/Characters/Player/Female/Medium/Bodies/F_MED_Bistro_Astronaut/FX/T_GaussianNoise_Normal}
  n8 = VectorParameter() {group=WPO, parameterName=WPO Wave UV, sortPriority=6, defaultValue=<rgba 1.5 1 -0.2 0.5>}
  n9 = Panner(Coordinate=Multiply(A=@n4, B=AppendVector(A=@n8.R, B=@n8.R)), Speed=AppendVector(A=@n8.B, B=@n8.B))
  n10 = AppendVector(
   A=Multiply(A=ComponentMask(None=@n7.RGB) {g=true, r=true}, B=ScalarParameter() {defaultValue=1, group=WPO, parameterName=WPO Side Intensity}) {constB=10}
   B=Multiply(
    A=TextureSample(UVs=@n9) {texture=&T_Smooth_tile}.R
    B=ScalarParameter() {defaultValue=1, group=WPO, parameterName=WPO Wave Intensity}
   ) {constB=10}
  )
  n11 = ComponentMask(None=Add(A=TextureCoordinate(), B=Constant2Vector() {r=0.1})) {r=true}
  Add(
   A=LinearInterpolate(
    A=Constant()
    B=MaterialFunctionCall(WidthBase=@n1, WidthTip=@n2, Expand U or V UV Channel=@n3) {materialFunction=&SplineThicken}.WorldPositionOffset
    Alpha=ScalarParameter() {parameterName=Use_SplineThicken}
   )
   B=Multiply(A=LinearInterpolate(B=@n10, Alpha=@n11), B=ScalarParameter() {parameterName=Final_WPO Intensity}) {constB=10}
  )
  n12 = VectorParameter() {parameterName=LerpColorA, defaultValue=<rgba 0.7552 0.0709 0.0151 1>}
  n13 = VectorParameter() {parameterName=LerpColorB, defaultValue=<rgba 5 1.381 0.1 1>}
  n14 = TextureCoordinate() {vTiling=0.5}
  n15 = Power(Base=Sine(None=ComponentMask(None=@n14) {g=true})) {constExponent=5}
  n16 = TextureCoordinate() {uTiling=1.5, vTiling=0.5}
  n17 = Constant() {r=0.05}
  n18 = Multiply(A=Time(), B=Constant() {r=0.2})
  n19 = TextureSampleParameter2D(UVs=Panner(
    Coordinate=MaterialFunctionCall(UVs=@n16, "Rotation Angle (0-1)"=@n17) {materialFunction=&CustomRotator}.Rotated Values
    Time=@n18
   ) {speedX=-1}) {
   parameterName=LightTextureA
   samplerType=SAMPLERTYPE_LinearGrayscale
   texture=/Game/Effects/Fort_Effects/Emotes/UltraLight/Texture/T_Plasma_D_Greyscale
  }
  n20 = Round(None=MaterialFunctionCall() {materialFunction=&LinearGradient}.VGradient)
  n21 = OneMinus(None=@n20)
  n22 = Add(A=TextureCoordinate() {uTiling=1.5, vTiling=0.5}, B=Constant2Vector() {r=0.15})
  n23 = Constant() {r=-0.05}
  n24 = TextureSampleParameter2D(UVs=Panner(
    Coordinate=MaterialFunctionCall(UVs=@n22, "Rotation Angle (0-1)"=@n23) {materialFunction=&CustomRotator}.Rotated Values
    Time=@n18
   ) {speedX=-1}) {
   parameterName=LightTextureB
   samplerType=SAMPLERTYPE_LinearGrayscale
   texture=&T_Plasma_D_Greyscale
  }
  n25 = LinearInterpolate(Alpha=Add(A=Multiply(A=@n19.R, B=@n21), B=Multiply(A=@n20, B=@n24.R))) {constB=1.5}
  n26 = ParticleColor()
  MaterialFunctionCall(Emissive=Multiply(A=LinearInterpolate(A=@n12.RGB, B=@n13.RGB, Alpha=@n15), B=Multiply(A=@n25, B=@n26.RGB))) {materialFunction=&MF_FX_EmissiveAdjustments}
  n27 = AppendVector(A=ScalarParameter() {defaultValue=1, parameterName=BaseTexture_U}, B=ScalarParameter() {defaultValue=0.5, parameterName=BaseTexture_V})
  n28 = TextureSampleParameter2D(UVs=Panner(Coordinate=TextureCoordinate() {uTiling=1.5, vTiling=1.5}) {speedX=-1}) {
   group=Distortion
   parameterName=NoiseTexture
   samplerType=SAMPLERTYPE_Normal
   sortPriority=0
   texture=&T_GaussianNoise_Normal
  }
  n29 = Multiply(A=ComponentMask(None=@n28.RGB) {g=true, r=true}, B=ScalarParameter() {defaultValue=0.02, parameterName=Noise_intensity})
  n30 = Add(A=Multiply(A=TextureCoordinate(), B=@n27), B=@n29)
  n31 = ScalarParameter() {defaultValue=0.05, parameterName=BaseTexA_Rotation}
  n32 = Multiply(A=Time(), B=ScalarParameter() {defaultValue=0.8, parameterName=BaseTexture_Speed})
  n33 = VectorParameter() {parameterName=BaseTextureA_XY, defaultValue=<rgba 0 0 -0.1 0.3>}
  n34 = TextureSampleParameter2D(UVs=Panner(
    Coordinate=MaterialFunctionCall(UVs=@n30, "Rotation Angle (0-1)"=@n31) {materialFunction=&CustomRotator}.Rotated Values
    Time=@n32
    Speed=AppendVector(A=@n33.B, B=@n33.B)
   ) {speedX=-0.1, speedY=0.3}) {
   parameterName=BaseTextureA
   samplerType=SAMPLERTYPE_LinearGrayscale
   texture=/Game/Effects/Fort_Effects/Emotes/UltraLight/Texture/T_DirectionalNoise_Greyscale
  }
  n35 = Round(None=MaterialFunctionCall() {materialFunction=&LinearGradient}.VGradient)
  n36 = Multiply(A=LinearInterpolate(Alpha=@n34.R) {constB=2}, B=OneMinus(None=@n35))
  n37 = Multiply(A=TextureCoordinate(), B=@n27)
  n38 = Constant2Vector() {r=0.15}
  n39 = MaterialFunctionCall(UVs=Add(A=@n29, B=Add(A=@n37, B=@n38)), "Rotation Angle (0-1)"=ScalarParameter() {defaultValue=-0.05, parameterName=BaseTexB_Rotation}) {materialFunction=&CustomRotator}
  n40 = VectorParameter() {parameterName=BaseTextureB_XY, defaultValue=<rgba 0 0 -0.1 0>}
  n41 = AppendVector(A=@n40.B, B=@n40.B)
  n42 = Multiply(
   A=@n35
   B=TextureSampleParameter2D(UVs=Panner(Coordinate=@n39.Rotated Values, Time=@n32, Speed=@n41) {speedX=-0.1}) {parameterName=BaseTextureB, texture=&/T_Pickaxe_BlessFlan_FireNoise}.R
  )
  n43 = MaterialFunctionCall(UVs=Add(A=TextureCoordinate() {uTiling=2, vTiling=0.5}), "Rotation Angle (0-1)"=Constant() {r=-0.02}) {materialFunction=&CustomRotator}
  n44 = Multiply(A=Time(), B=Constant() {r=0.8})
  n45 = VectorParameter() {parameterName=MaskTextureA_XY, defaultValue=<rgba 0 0 -0.3 0.5>}
  n46 = AppendVector(A=@n45.B, B=@n45.B)
  n47 = Round(None=MaterialFunctionCall() {materialFunction=&LinearGradient}.VGradient)
  n48 = Multiply(
   A=TextureSampleParameter2D(UVs=Panner(Coordinate=@n43.Rotated Values, Time=@n44, Speed=@n46) {speedX=-0.3, speedY=0.5}) {parameterName=MaskTextureA, texture=&/T_Pickaxe_BlessFlan_Mask}.R
   B=OneMinus(None=@n47)
  )
  n49 = TextureCoordinate() {uTiling=2, vTiling=0.5}
  n50 = Constant2Vector() {r=0.15}
  n51 = MaterialFunctionCall(UVs=Add(B=Add(A=@n49, B=@n50)), "Rotation Angle (0-1)"=Constant() {r=0.3}) {materialFunction=&CustomRotator}
  n52 = VectorParameter() {parameterName=MaskTextureB_XY, defaultValue=<rgba 0 0 -0.3 0>}
  n53 = AppendVector(A=@n52.B, B=@n52.B)
  n54 = Multiply(
   A=@n47
   B=TextureSampleParameter2D(UVs=Panner(Coordinate=@n51.Rotated Values, Time=@n44, Speed=@n53) {speedX=-0.3}) {parameterName=MaskTextureB, texture=&T_Pickaxe_BlessFlan_Mask}.R
  )
  n55 = Multiply(A=LinearInterpolate(Alpha=Add(A=@n36, B=@n42)) {constA=-0.3}, B=LinearInterpolate(Alpha=Add(A=@n48, B=@n54)) {constB=1.5})
  n56 = TextureCoordinate()
  n57 = ComponentMask(None=@n56) {r=true}
  n58 = OneMinus(None=@n57)
  n59 = Power(Base=Multiply(A=Multiply(A=@n57, B=@n58), B=ScalarParameter() {defaultValue=10, parameterName=R_Mask_Multiply}) {constB=5}) {constExponent=6}
  n60 = ComponentMask(None=@n56) {g=true}
  n61 = OneMinus(None=@n60)
  n62 = Power(
   Base=Multiply(A=Multiply(A=@n60, B=@n61), B=ScalarParameter() {defaultValue=10, parameterName=G_Mask_Multiply}) {constB=6}
   Exp=ScalarParameter() {defaultValue=6, parameterName=G_Mask_Power}
  ) {constExponent=3}
  n63 = TextureCoordinate() {uTiling=0.4}
  n64 = Constant2Vector() {g=0.5, r=0.4}
  n65 = TextureCoordinate() {uTiling=0.4}
  n66 = Constant2Vector() {g=0.5}
  n67 = Multiply(
   A=Multiply(A=Saturate(None=@n59), B=Saturate(None=@n62))
   B=Multiply(
    A=SphereMask(A=@n63, B=@n64) {attenuationRadius=0.55, hardnessPercent=50}
    B=SphereMask(A=@n65, B=@n66) {attenuationRadius=0.55, hardnessPercent=50}
   )
  )
  n68 = ComponentMask(None=TextureCoordinate()) {g=true}
  n69 = OneMinus(None=@n68)
  n70 = Power(Base=Multiply(A=Multiply(A=@n68, B=@n69)) {constB=4.2}) {constExponent=5}
  n71 = Multiply(A=Multiply(A=Multiply(A=@n26.A, B=@n55), B=Saturate(None=@n67)), B=Saturate(None=OneMinus(None=@n70)))
  n72 = TextureCoordinate()
  n73 = Constant2Vector() {r=-0.05}
  n74 = ComponentMask(None=MaterialFunctionCall(UVs=Add(A=@n72, B=@n73), "Rotation Angle (0-1)"=Constant() {r=0.44}) {materialFunction=&CustomRotator}.Rotated Values) {r=true}
  n75 = Saturate(None=Power(Base=Multiply(A=@n74) {constB=3}) {constExponent=7})
  Multiply(
   A=LinearInterpolate(A=@n71, B=Multiply(A=@n71, B=@n75), Alpha=ScalarParameter() {parameterName=Use_CornerMask})
   B=MaterialFunctionCall("FadeSensitivity (optional)"=ScalarParameter() {defaultValue=1, group=Opacity, parameterName=MeshFade}) {materialFunction=&FadeBasedOnViewAngle}.Result
  )
 ]
}
